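; PlayerDynamicCapsuleComponent is a Blueprint event graph, named as a dynamic capsule component for the player.
; It declares BeginPlay, Tick and two debug key events for F1 and F2, all with empty bodies.
; The graph holds no capsule logic, so any such behavior is not visible here.
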(event EventBeginPlay)

(event EventTick
  (params (DeltaSeconds :type Float%20%28single-precision%29)))

(event DebugKeyF1
  (params (Key :type Key%20Structure) (ActionValue :type Input%20Action%20Value%20Structure)))

(event DebugKeyF2
  (params (Key :type Key%20Structure) (ActionValue :type Input%20Action%20Value%20Structure)))
